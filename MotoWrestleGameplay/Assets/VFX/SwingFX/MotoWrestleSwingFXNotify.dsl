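; Niagara anim notify for MotoWrestle swing VFX that configures the effect spawned when an animation reaches it.
; It calls the parent notify, casts the spawned effect to a Niagara component and returns the parent's result.
; Its own variables override the system's user parameters: a target static mesh, a mesh index, and tuning for energy shape, ice, particles and smoke.
; The target mesh is assigned through both the direct and variable static-mesh setters.
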(fn Received_Notify
  (params (MeshComp :type Skeletal%20Mesh%20Component%20Object%20Reference) (Animation :type Anim%20Sequence%20Base%20Object%20Reference) (EventReference :type Anim%20Notify%20Event%20Reference%20Structure%20%28by%20ref%29))
  (bind _targetmesh (Variables|Default|GetTargetMesh))
  (bind _returnvalue (|CallParentFunction MeshComp Animation EventReference))
  (bind _returnvalue_2 (AnimNotify|GetSpawnedEffect))
  (bind _asniagara_particle_system_component (Utilities|Casting|CastToNiagaraComponent _returnvalue_2))
  (Niagara|SetNiagaraStaticMeshDirectly _asniagara_particle_system_component "TargetStaticMesh" _targetmesh)
  (Niagara|SetNiagaraVariable(StaticMesh) _asniagara_particle_system_component "TargetStaticMesh" _targetmesh)
  (Niagara|SetNiagaraVariable(Vector2) _asniagara_particle_system_component "FollowParticle_VelocityRange" (Variables|Default|GetFollowParticle_VelocityRange))
  (Niagara|SetNiagaraVariable(Vector2) _asniagara_particle_system_component "FollowSmoke_LifeRange" (Variables|Default|GetFollowSmoke_LifeRange))
  (Niagara|SetNiagaraVariable(Vector2) _asniagara_particle_system_component "FollowSmoke_SpriteSizeRange" (Variables|Default|GetFollowSmoke_SpriteSizeRange))
  (Niagara|SetNiagaraVariable(Vector2) _asniagara_particle_system_component "FollowSmoke_VelocityRange" (Variables|Default|GetFollowSmoke_VelocityRange))
  (Niagara|SetNiagaraVariable(Float) _asniagara_particle_system_component "ExtraEnergy" (Variables|Default|GetExtraEnergy))
  (Niagara|SetNiagaraVariable(Float) _asniagara_particle_system_component "FollowIcePiece_SpawnRate" (Variables|Default|GetFollowIcePiece_SpawnRate))
  (Niagara|SetNiagaraVariable(Float) _asniagara_particle_system_component "FollowParticle_SpawnRate" (Variables|Default|GetFollowParticle_SpawnRate))
  (Niagara|SetNiagaraVariable(Float) _asniagara_particle_system_component "FollowSmoke_InheritV" (Variables|Default|GetFollowSmoke_InheritV))
  (Niagara|SetNiagaraVariable(Float) _asniagara_particle_system_component "FollowSmoke_SpawnRate" (Variables|Default|GetFollowSmoke_SpawnRate))
  (Niagara|SetNiagaraVariable(Float) _asniagara_particle_system_component "LargeIceSpike" (Variables|Default|GetLargeIceSpike))
  (Niagara|SetNiagaraVariable(Float) _asniagara_particle_system_component "ScaleEnergyWidth" (Variables|Default|GetScaleEnergyWidth))
  (Niagara|SetNiagaraVariable(Float) _asniagara_particle_system_component "SideViewOffset" (Variables|Default|GetSideViewOffset))
  (Niagara|SetNiagaraVariable(Float) _asniagara_particle_system_component "TipWave" (Variables|Default|GetTipWave))
  (Niagara|SetNiagaraVariable(Int32) _asniagara_particle_system_component "MeshIndex" (Variables|Default|GetMeshIndex))
  (return _returnvalue))
