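# Niagara system for the end burst of the third swing effect in MotoWrestle, judging by its name and path.
# It combines five emitters named Blast, Glow, IcePiece, Smoke001 and Spark under a SystemState module, with no renderers or emitter modules listed.
# An integer user parameter, User.ColorVariant, is exposed, presumably to select a color variant, though no use of it appears here.
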NiagaraSystem /MotoWrestleGameplay/Assets/VFX/SwingFX/NS_MotoWrestle_Swing3_EndBurst
dependencies {usedDataInterfaces [], usedDynamicInputs [], usedModules [&SystemState]}
emitters {
 Blast {rendererClasses [], renderers []}
 Glow {rendererClasses [], renderers []}
 IcePiece {rendererClasses [], renderers []}
 Smoke001 {rendererClasses [], renderers []}
 Spark {rendererClasses [], renderers []}
}
summary {
 systemName=NS_MotoWrestle_Swing3_EndBurst
 emitters [
  {emitterName=Smoke001, rendererClasses []}
  {emitterName=Spark, rendererClasses []}
  {emitterName=IcePiece, rendererClasses []}
  {emitterName=Blast, rendererClasses []}
  {emitterName=Glow, rendererClasses []}
 ]
 userVariables [User.ColorVariant {type=NiagaraInt32}]
}
